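# UI material for the HUD hit-tick shield-break effect, drawing a cracking shield icon with a glow.
# The red and green channels of one texture slide in opposite directions by BreakAngle and BreakAmount, scaled by ShieldScale.
# A blue-to-cyan gradient from shared collection colours tints them, plus a blue circle-SDF glow shaped by a curve atlas.
# Vertex alpha fades the sum; the yellow Switch_Crit blend does not feed it.
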Material /BattleRoyaleGameUI/HUD/HitTicks/Assets/M_UI_ShieldBreak
graph {
 properties {blendMode=BLEND_AlphaComposite, materialDomain=MD_UI, pixelDepthOffsetMode=PDOM_Legacy, usedShadingModels=Default Lit}
 nodes [
  n1 = ComponentMask(None=CollectionParameter() {collection=/BattleRoyaleGameUI/Shared/MPC_BRGameUI, parameterName=Blue}) {b=true, g=true, r=true}
  n2 = ComponentMask(None=CollectionParameter() {collection=&MPC_BRGameUI, parameterName=Cyan}) {b=true, g=true, r=true}
  n3 = MaterialFunctionCall() {materialFunction=&GetUserInterfaceUV}
  n4 = ScalarParameter() {defaultValue=0.5, parameterName=ShieldScale}
  n5 = MaterialFunctionCall(UVs=@n3.Normalized UV, Texture Scale=@n4) {materialFunction=&ScaleUVsByCenter}
  n6 = LinearInterpolate(
   A=LinearInterpolate(A=@n1, B=@n2, Alpha=ScalarParameter() {defaultValue=0.15, parameterName=GradientFlatness})
   B=@n2
   Alpha=ComponentMask(None=@n5.UVs) {r=true}
  )
  LinearInterpolate(
   A=@n6
   B=ComponentMask(None=CollectionParameter() {collection=&MPC_BRGameUI, parameterName=Yellow}) {b=true, g=true, r=true}
   Alpha=ScalarParameter() {parameterName=Switch_Crit}
  )
  n7 = CurveAtlasRowParameter(CurveTime=SmoothStep(Value=MaterialFunctionCall!(UVs, Radius) {materialFunction=/FortUILibrary/MaterialFunctions/Shapes/SDFs/MF_SDF_Circle}.SDF) {constMax=-0.65, constMin=0.75}) {
   atlas=/BattleRoyaleGameUI/Shared/UICurveAtlas_BRGameUI
   curve=/BattleRoyaleGameUI/Shared/UICurve_BRGameUI_StandardFalloffs
   parameterName=GlowCurve
  }
  n8 = Multiply(A=@n7.A, B=@n1)
  n9 = ScalarParameter() {parameterName=GlowIntensity}
  n10 = ScalarParameter() {defaultValue=0.15, parameterName=BreakAngle}
  n11 = ScalarParameter() {parameterName=BreakAmount}
  n12 = AppendVector(A=Multiply(A=Sine(None=@n10), B=@n11), B=Multiply(A=Cosine(None=@n10), B=@n11))
  n13 = MaterialFunctionCall(UVs=@n3.Normalized UV, Texture Scale=@n4) {materialFunction=&ScaleUVsByCenter}
  n14 = Max(
   A=TextureSample(UVs=Add(A=@n5.UVs, B=@n12)) {samplerSource=SSM_Clamp_WorldGroupSettings, texture=&/T_UI_ShieldBreak}.R
   B=TextureSample(UVs=Subtract(A=@n13.UVs, B=@n12)) {samplerSource=SSM_Clamp_WorldGroupSettings, texture=&T_UI_ShieldBreak}.G
  )
  Multiply(A=VertexColor().A, B=Add(A=Multiply(A=@n8, B=@n9), B=Multiply(A=@n6, B=@n14)))
 ]
}
