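# Battle Royale HUD hit-tick widget that animates hit, elimination, shield-hit and shield-break feedback.
# Hit events feed damage, random rotation and crit state into dynamic materials, then play elimination, shield or normal hit animations, skipping shield hits during a break.
# Shield-break events play their own animation; animations mostly drive material parameters.
# Animation start/finish callbacks wait a tick, then show the widget only while any animation plays.
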WidgetBlueprint /BattleRoyaleGameUI/HUD/HitTicks/WBP_BRGameUI_HitTicksWidget
animations [
 AnimHit {
  binding_count=1
  length_seconds=0.3334
  object=&self:AnimHit
  sequence {
   bindings [HitMarker {
     id=#1
     possessed_class=Image
     tracks [{
       type=MovieSceneWidgetMaterialTrack
       sections [{
         type=MovieSceneParameterSection
         end_frame=10
         start_frame=0
         channels [
          AnimSwitch_Elim {type=MovieSceneScriptingFloatChannel, const [0,1]}
          Anim_TickSpread {type=MovieSceneScriptingFloatChannel, keys [[0,0],[3,1],[10,0.0033]]}
          Anim_TickWidth {type=MovieSceneScriptingFloatChannel, keys [[0,0],[3,0],[10,1]]}
          Switch_Anim {type=MovieSceneScriptingFloatChannel, thinned_from=3, keys [[0,1],[10,0]]}
          Anim_Opacity {type=MovieSceneScriptingFloatChannel, keys [[0,1],[4,1],[10,0]]}
         ]
        }]
      }]
    }]
   display_rate {numerator=30}
  }
 }
 AnimElim {
  binding_count=1
  length_seconds=0.75
  object=&self:AnimElim
  sequence {bindings [HitMarker {
     id=#1
     possessed_class=Image
     tracks [{
       type=MovieSceneWidgetMaterialTrack
       sections [{
         type=MovieSceneParameterSection
         end_frame=15
         start_frame=0
         channels [
          Switch_Anim {type=MovieSceneScriptingFloatChannel, keys [[0,1],[15,0]]}
          AnimSwitch_Elim {type=MovieSceneScriptingFloatChannel, keys [[0,1],[15,0]]}
          Anim_TickSpread {type=MovieSceneScriptingFloatChannel, keys [[0,0],[3,1],[13,1],[15,0]]}
          Anim_TickScale {type=MovieSceneScriptingFloatChannel, keys [[0,1],[9,1],[15,0]]}
          Switch_Fatal {type=MovieSceneScriptingFloatChannel, keys [[0,1],[3,0],[6,1],[14,1],[15,0]]}
          Anim_MarkScale {type=MovieSceneScriptingFloatChannel, keys [[0,1],[3,0],[5,0.85],[7,0.85],[15,0]]}
          Anim_Opacity {type=MovieSceneScriptingFloatChannel, keys [[0,1],[7,1],[14,0],[15,0]]}
          Anim_TickWidth {type=MovieSceneScriptingFloatChannel, keys [[0,0],[3,-0.5],[14.9997,0]]}
         ]
        }]
      }]
    }]}
 }
 AnimShieldHit {
  binding_count=2
  length_seconds=0.5
  object=&self:AnimShieldHit
  sequence {bindings [
    ShieldIcon {
     id=#2
     possessed_class=Image
     tracks [
      {
       type=MovieSceneFloatTrack
       display_name=Render Opacity
       sections [{type=MovieSceneFloatSection, end_frame=10, start_frame=0, channels [None {type=MovieSceneScriptingFloatChannel, keys [[0,1],[10,0]]}]}]
      }
      {
       type=MovieSceneWidgetMaterialTrack
       sections [{
         type=MovieSceneParameterSection
         end_frame=10
         start_frame=0
         channels [
          BreakAmount {type=MovieSceneScriptingFloatChannel, const [0,1]}
          BreakAngle {type=MovieSceneScriptingFloatChannel, const [0.15,1]}
          GlowIntensity {type=MovieSceneScriptingFloatChannel, keys [[0,1],[5.9997,0]]}
         ]
        }]
      }
      {
       type=MovieScene2DTransformTrack
       sections [{
         type=MovieScene2DTransformSection
         end_frame=9
         start_frame=0
         channels [
          Translation.X {type=MovieSceneScriptingFloatChannel, default=30}
          Translation.Y {type=MovieSceneScriptingFloatChannel, default=0}
          Angle {type=MovieSceneScriptingFloatChannel, const [0,1]}
          Scale.X {type=MovieSceneScriptingFloatChannel, default=1.25, const [1.25,1]}
          Scale.Y {type=MovieSceneScriptingFloatChannel, default=1.25, const [1.25,1]}
          Shear.X {type=MovieSceneScriptingFloatChannel, default=0}
          Shear.Y {type=MovieSceneScriptingFloatChannel, default=0}
         ]
        }]
      }
     ]
    }
    ShieldMarker {
     id=#3
     possessed_class=Image
     tracks [{
       type=MovieSceneWidgetMaterialTrack
       sections [{
         type=MovieSceneParameterSection
         end_frame=10
         start_frame=0
         channels [
          Switch_Anim {type=MovieSceneScriptingFloatChannel, const [1,1]}
          Switch_ShieldBreak {type=MovieSceneScriptingFloatChannel, const [0,1]}
          Anim_TickSpread {type=MovieSceneScriptingFloatChannel, keys [[0,1],[3,0],[10,1]]}
          Anim_Opacity {type=MovieSceneScriptingFloatChannel, keys [[0,1],[4,1],[10,0]]}
          Anim_TickWidth {type=MovieSceneScriptingFloatChannel, keys [[0,0],[2,0],[10,1]]}
         ]
        }]
      }]
    }
   ]}
 }
 AnimShieldBreak {
  binding_count=2
  length_seconds=0.5
  object=&self:AnimShieldBreak
  sequence {bindings [
    ShieldIcon {
     id=#4
     possessed_class=Image
     tracks [
      {
       type=MovieSceneWidgetMaterialTrack
       sections [{
         type=MovieSceneParameterSection
         end_frame=10
         start_frame=0
         channels [
          BreakAmount {type=MovieSceneScriptingFloatChannel, keys [[0,0],[3,0.2],[8,0.15]]}
          BreakAngle {type=MovieSceneScriptingFloatChannel, keys [[0,0.1],[8,0.15]]}
          GlowIntensity {type=MovieSceneScriptingFloatChannel, keys [[0,0],[3,1.5],[8,0]]}
         ]
        }]
      }
      {
       type=MovieScene2DTransformTrack
       sections [{
         type=MovieScene2DTransformSection
         end_frame=10
         start_frame=0
         channels [
          Translation.X {type=MovieSceneScriptingFloatChannel, default=30}
          Translation.Y {type=MovieSceneScriptingFloatChannel, default=0}
          Angle {type=MovieSceneScriptingFloatChannel, keys [[0,0],[8,-12]]}
          Scale.X {type=MovieSceneScriptingFloatChannel, default=1.25, keys [[0,1.25],[2,2],[8,1.25]]}
          Scale.Y {type=MovieSceneScriptingFloatChannel, default=1.25, keys [[0,1.25],[2,2],[8,1.25]]}
          Shear.X {type=MovieSceneScriptingFloatChannel, default=0}
          Shear.Y {type=MovieSceneScriptingFloatChannel, default=0}
         ]
        }]
      }
      {
       type=MovieSceneFloatTrack
       display_name=Render Opacity
       sections [{type=MovieSceneFloatSection, end_frame=10, start_frame=0, channels [None {type=MovieSceneScriptingFloatChannel, keys [[5,1],[10,0]]}]}]
      }
     ]
    }
    ShieldMarker {
     id=#5
     possessed_class=Image
     tracks [{
       type=MovieSceneWidgetMaterialTrack
       sections [{
         type=MovieSceneParameterSection
         end_frame=10
         start_frame=0
         channels [
          Switch_Shield {type=MovieSceneScriptingFloatChannel, const [1,2]}
          Switch_Anim {type=MovieSceneScriptingFloatChannel, const [1,2]}
          Switch_ShieldBreak {type=MovieSceneScriptingFloatChannel, thinned_from=4, keys [[0,0],[1,1],[9.9997,1]]}
          Anim_TickSpread {type=MovieSceneScriptingFloatChannel, thinned_from=5, keys [[0,1],[2,0],[9.9997,0],[10,1]]}
          Anim_Opacity {type=MovieSceneScriptingFloatChannel, keys [[0,1],[6,1],[9.9997,0]]}
          Anim_ShieldBreak {type=MovieSceneScriptingFloatChannel, keys [[0,0],[1,1],[6,1],[10,0]]}
          Anim_MarkScale {type=MovieSceneScriptingFloatChannel, keys [[0,0],[1,-1],[6,0],[10,0]]}
          Switch_Crit {type=MovieSceneScriptingFloatChannel, const [0,2]}
         ]
        }]
      }]
    }
   ]}
 }
]
graphs [
 EventGraph {nodes [
   n1 = OnAnyAnimationPlayed() {is=title K2Node_CustomEvent}
   > DelayUntilNextTick!(execute) -> (then)
   > n2 = Branch(Condition=IsAnyAnimationPlaying!(self:User Widget Object Reference) -> (ReturnValue:bool))
   > SetVisibility(InVisibility=HitTestInvisible)
   SetVisibility(execute=@n2.else, InVisibility=Hidden)
   Event OnInitialized()
   > n3 = Set Array(Array:Array of Widget Animation Object References=Make Array!(
     "[0]":Widget Animation Object Reference=Get AnimHit() -> (AnimHit:Widget Animation Object Reference)
     "[1]":Widget Animation Object Reference=Get AnimElim() -> (AnimElim:Widget Animation Object Reference)
     "[2]":Widget Animation Object Reference=Get AnimShieldHit() -> (AnimShieldHit:Widget Animation Object Reference)
     "[3]":Widget Animation Object Reference=Get AnimShieldBreak() -> (AnimShieldBreak:Widget Animation Object Reference)
    ) -> (Array:Array of Widget Animation Object References)) -> (Output_Get:Array of Widget Animation Object References)
   > n4 = For Each Loop(Array:Array of Widget Animation Object References=@n3) -> (Array Element:Widget Animation Object Reference)
   > BindToAnimationStarted!(execute, self:User Widget Object Reference, Animation:Widget Animation Object Reference=@n4, Delegate:Delegate=@n1) -> (then)
   > BindToAnimationFinished(Animation=@n4, Delegate=@n1)
   Event Construct()
   > SetVisibility(InVisibility=Hidden)
   > n5 = Cast To FortPlayerController(Object=GetOwningPlayer())
   > Set As Fort Player Controller(As Fort Player Controller:Fort Player Controller Object Reference=@n5) -> (Output_Get:Fort Player Controller Object Reference)
   > n6 = GetDynamicMaterial(self=Get ShieldIcon() -> (ShieldIcon:Image Object Reference))
   > Set M_ShieldIcon(M_ShieldIcon:Material Instance Dynamic Object Reference=@n6) -> (Output_Get:Material Instance Dynamic Object Reference)
   > n7 = GetDynamicMaterial(self=Get ShieldMarker() -> (ShieldMarker:Image Object Reference))
   > Set M_ShieldMarker(M_ShieldMarker:Material Instance Dynamic Object Reference=@n7) -> (Output_Get:Material Instance Dynamic Object Reference)
   > n8 = GetDynamicMaterial(self=Get HitMarker() -> (HitMarker:Image Object Reference))
   > Set M_HitMarker(M_HitMarker:Material Instance Dynamic Object Reference=@n8) -> (Output_Get:Material Instance Dynamic Object Reference)
   > Broadcast Field Value Changed!(execute, self:Widget Object Reference, FieldId:Field Notification Id Structure=(FieldName="UseNativeMarkerTag")) -> (then)
   > Set HitInterval!(
    execute
    HitInterval:double=Get HitInterval() -> (HitInterval:double)
    self:Fort Athena Hit Ticks View Model Object Reference=Get FortAthenaHitTicksViewModel() -> (FortAthenaHitTicksViewModel:Fort Athena Hit Ticks View Model Object Reference)
   ) -> (then, Output_Get:double)
  ]}
 HandleHitTickEvent {nodes [
   n1 = HandleHitTickEvent!() -> (
    then
    Event_Damage:double=~
    Event_AccumulatedDamage:double=~
    Event_bShield:bool=~
    Event_bCritical:bool=~
    Event_bFatal:bool=~
    Event_Target:Actor Object Reference
   ) {is=title K2Node_FunctionEntry}
   > n2 = Set DamageFloat(DamageFloat:double=@n1) -> (Output_Get:double)
   n3 = Get M_HitMarker() -> (M_HitMarker:Material Instance Dynamic Object Reference)
   > SetScalarParameterValue(self=@n3, ParameterName=Float_DamageMultipler, Value=@n2)
   > SetScalarParameterValue(
    self=[Get M_ShieldMarker() -> (M_ShieldMarker:Material Instance Dynamic Object Reference), @n3]
    ParameterName=HitMark_RandomHitRotation
    Value=RandomFloat!() -> (ReturnValue:double)
   )
   > n4 = Branch(Condition=@n1.Event_bCritical)
   n5 = Get M_ShieldMarker()
   n6 = Get M_HitMarker()
   > n7 = SetScalarParameterValue(self=[@n5, @n6], ParameterName=Switch_Crit, Value=1)
   n8 = Branch(execute=[@n7.then, @n11.then], Condition=@n1.Event_bFatal)
   > SetScalarParameterValue(self=Get M_HitMarker(), ParameterName=Float_DamageMultipler, Value=0)
   > StopAnimation(InAnimation=Get AnimHit())
   > PlayAnimation(InAnimation=Get AnimElim())
   n9 = Branch(execute=@n8.else, Condition=@n1.Event_bShield)
   > n10 = Branch(Condition=IsAnimationPlaying(InAnimation=Get AnimShieldBreak()))
   PlayAnimation(execute=@n10.else, InAnimation=Get AnimShieldHit())
   PlayAnimation(execute=@n9.else, InAnimation=Get AnimHit())
   n11 = SetScalarParameterValue(execute=@n4.else, self=[@n5, @n6], ParameterName=Switch_Crit, Value=0)
   Get M_ShieldIcon() -> (M_ShieldIcon:Material Instance Dynamic Object Reference)
  ]}
 HandleShieldBreakEvent {nodes [
   HandleShieldBreakEvent!() -> (then, bIsOvershield:bool=~) {is=title K2Node_FunctionEntry}
   > PlayAnimation(InAnimation=Get AnimShieldBreak())
  ]}
]
variables [
 Array {type="array<WidgetAnimation>", default []}
 As Fort Player Controller {type=FortPlayerController, default=null}
 CumulativeDamageDuration {type=number, default=1}
 CumulativeDamageFloat {type=number}
 HitInterval {type=number, default=3}
 LastHitTarget {type=Actor, default=null}
 LastHitTime {type=number}
 M_HitMarker {type=MaterialInstanceDynamic}
 M_ShieldIcon {type=MaterialInstanceDynamic}
 M_ShieldMarker {type=MaterialInstanceDynamic}
 MaxDamage {type=ScalableFloat, default {registryType=None, value=0, curve {curveTable=null, rowName=None}}}
 OffsetTimer {type=TimerHandle}
 Tint Color {type=SlateColor, default {colorUseRule=UseColor_Specified, specifiedColor=<rgba 1 0 0.9835 1>}}
 UseNativeMarkerTag {type=GameplayTag, default {tagName=Item.Weapon.Ranged.Shotgun}}
]
widgets {
 info {
  inheritedWidgetCount=1
  namedSlotCount=1
  parentClass=/Script/HitTicksUI.FortHitTicksHUDElement
  rootWidgetClass=&Overlay
  widgetCount=7
 }
 widgets [
  Overlay_1 {widgetClassPath=&Overlay, widget=Overlay_1}
  FortHUDElementWrapper {
   widgetClassPath=/Game/Athena/HUD/FortHUDElementWrapper
   bIsVariable=true
   parent=Overlay_1
   widget=FortHUDElementWrapper
   properties {
    elementTag {tagName=HUD.Athena.HitTicks}
    hUDElementTag {gameplayTags [{tagName=Athena.HideForCapture}, {tagName=HUD.Athena.HideForPostGame}]}
   }
   slot {class=OverlaySlot}
  }
  Overlay_0 {widgetClassPath=&Overlay, namedSlotHost=FortHUDElementWrapper, widget=Overlay_0}
  ShieldMarker {
   widgetClassPath=&Image
   bIsVariable=true
   parent=Overlay_0
   widget=ShieldMarker
   properties {
    visibility=SelfHitTestInvisible
    brush {
     bIsDynamicallyLoaded=false
     drawAs=Image
     imageType=FullColor
     mirroring=NoMirror
     resourceName=None
     resourceObject=/BattleRoyaleGameUI/HUD/HitTicks/Assets/MI_UI_ShieldTicks
     tiling=NoTile
     imageSize=<90 90>
     margin=<ltrb 0 0 0 0>
     outlineSettings {
      bUseBrushTransparency=false
      roundingType=HalfHeightRadius
      width=0
      color {colorUseRule=UseColor_Specified, specifiedColor=<rgba 0 0 0 0>}
      cornerRadii=<0 0 0 0>
     }
     tintColor {colorUseRule=UseColor_Specified, specifiedColor=<rgba 1 1 1 1>}
     uVRegion {bIsValid=false, max=<0 0>, min=<0 0>}
    }
   }
   slot {class=OverlaySlot}
  }
  ShieldIcon {
   widgetClassPath=&Image
   bIsVariable=true
   parent=Overlay_0
   widget=ShieldIcon
   properties {
    visibility=Hidden
    brush {
     bIsDynamicallyLoaded=false
     drawAs=Image
     imageType=FullColor
     mirroring=NoMirror
     resourceName=None
     resourceObject=/BattleRoyaleGameUI/HUD/HitTicks/Assets/M_UI_ShieldBreak
     tiling=NoTile
     imageSize=<36 36>
     margin=<ltrb 0 0 0 0>
     outlineSettings {
      bUseBrushTransparency=false
      roundingType=HalfHeightRadius
      width=0
      color {colorUseRule=UseColor_Specified, specifiedColor=<rgba 0 0 0 0>}
      cornerRadii=<0 0 0 0>
     }
     tintColor {colorUseRule=UseColor_Specified, specifiedColor=<rgba 1 1 1 1>}
     uVRegion {bIsValid=false, max=<0 0>, min=<0 0>}
    }
    renderTransform {angle=0, scale=<1.25 1.25>, shear=<0 0>, translation=<30 0>}
   }
   slot {class=OverlaySlot, properties {horizontalAlignment=HAlign_Right, verticalAlignment=VAlign_Center}}
  }
  HitMarker {
   widgetClassPath=&Image
   bIsVariable=true
   parent=Overlay_0
   widget=HitMarker
   properties {
    visibility=SelfHitTestInvisible
    brush {
     bIsDynamicallyLoaded=false
     drawAs=Image
     imageType=FullColor
     mirroring=NoMirror
     resourceName=None
     resourceObject=/BattleRoyaleGameUI/HUD/HitTicks/Assets/MI_UI_HitTicks
     tiling=NoTile
     imageSize=<90 90>
     margin=<ltrb 0 0 0 0>
     outlineSettings {
      bUseBrushTransparency=false
      roundingType=HalfHeightRadius
      width=0
      color {colorUseRule=UseColor_Specified, specifiedColor=<rgba 0 0 0 0>}
      cornerRadii=<0 0 0 0>
     }
     tintColor {colorUseRule=UseColor_Specified, specifiedColor=<rgba 1 1 1 1>}
     uVRegion {bIsValid=false, max=<0 0>, min=<0 0>}
    }
   }
   slot {class=OverlaySlot}
  }
  TutorialHighlight {widgetClassPath=&TutorialHighlight}
 ]
}
